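; Far-lights volume Blueprint that draws instanced-mesh glows standing in for light actors inside its box.
; RefreshLightActors caches overlapping actors' point-light data, optionally rejecting interior lights with an upward line or sphere trace, then rebuilds instances.
; Construction spawns instances from that cache, scaled by attenuation radius and per-class overrides.
; Shared look settings go through custom primitive data, reapplied after BeginPlay; per-light color, intensity and overrides use per-instance custom data.
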(event EventBeginPlay
  (Utilities|FlowControl|DelayUntilNextTick)
  (FarLights|SetISMComponentParameters))

(fn CalculateInstanceTransform
  (params (LightActorCachedData :type S%20Far%20Light%20Actors%20Cached%20Data%20Structure))
  (bind _lightactorcacheddata (Variables|GetLightActorCachedData))
  (bind (_lightactorclass_2_79fd611942b4aeaa5582ce8dbcdcf58d _location_5_9ac4b656417ca6e93566e9b85854801a _attenuationradius_8_9dd871f04bd00c5003fc7e95840493c5) (Utilities|Struct|BreakSFarLightActorsCachedData _lightactorcacheddata))
  (bind _perclassoverrides (Variables|Parameters|GetPerClassOverrides))
  (bind (_value _returnvalue) (Utilities|Map|Find _perclassoverrides _lightactorclass_2_79fd611942b4aeaa5582ce8dbcdcf58d))
  (Utilities|FlowControl|Sequence
    (:then_0
      (if _returnvalue
        (LocalVariables|SetSizeMultiplier (Utilities|Struct|BreakSFarLightOverrideParams _value))
        (else
          (LocalVariables|SetSizeMultiplier 1.0))))
    (:then_1
      (return _location_5_9ac4b656417ca6e93566e9b85854801a "0, 0, 0" (Math|Conversions|ToVector(Float) (* _attenuationradius_8_9dd871f04bd00c5003fc7e95840493c5 (* (* (Variables|Parameters|GetSize) (LocalVariables|GetSizeMultiplier)) 0.002)))))))

(fn GatherLightActorData
  (Utilities|Array|Clear (Variables|DontEdit|GetCachedLightActorsData))
  (bind _outactors (Collision|ComponentOverlapActors (Variables|Default|GetBox) (Transformation|GetWorldTransform (Variables|Default|GetBox)) (Utilities|Array|MakeArray "ObjectTypeQuery1")))
  (for _array_element _outactors
    (Utilities|IsValid (Actor|GetComponentbyClass _array_element "/Script/Engine.PointLightComponent")
      (:Is%20Valid
        _)
      (:Is%20Not%20Valid
        (Utilities|IsValid (Actor|GetComponentbyClass "/Script/Engine.SpotLightComponent")
          (:Is%20Valid
            _))))
    (bind _returnvalue (FarLights|IsInteriorLight _array_element (Actor|GetComponentbyClass _array_element "/Script/Engine.PointLightComponent")))
    (if (not _returnvalue)
      (Utilities|Array|Add (Variables|DontEdit|GetCachedLightActorsData) (Utilities|Struct|MakeSFarLightActorsCachedData (Utilities|GetClass _array_element) (Transformation|GetWorldLocation (Actor|GetComponentbyClass _array_element "/Script/Engine.PointLightComponent")) (Variables|/Script/Engine.LocalLightComponent|GetAttenuationRadius (Actor|GetComponentbyClass _array_element "/Script/Engine.PointLightComponent")) (Variables|/Script/Engine.LightComponentBase|GetIntensity (Actor|GetComponentbyClass _array_element "/Script/Engine.PointLightComponent")) (Rendering|Components|Light|GetLightColor (Actor|GetComponentbyClass _array_element "/Script/Engine.PointLightComponent")))))))

(fn GenerateMeshIstances
  (bind _instancedstaticmesh (Variables|Default|GetInstancedStaticMesh))
  (for _array_element (Variables|DontEdit|GetCachedLightActorsData)
    (bind _returnvalue (Components|InstancedStaticMesh|AddInstance _instancedstaticmesh (FarLights|CalculateInstanceTransform _array_element) true))
    (FarLights|SetPerInstanceMaterialParameters _instancedstaticmesh _returnvalue _array_element)))

(fn IsInteriorLight
  (params (LightActor :type Actor%20Object%20Reference) (LightComponent :type Point%20Light%20Component%20Object%20Reference))
  (bind _returnvalue (Transformation|GetWorldLocation LightComponent))
  (bind _array (Utilities|Array|MakeArray LightActor))
  (bind _ceilingdetectiondistance (Variables|Parameters|Technical|GetCeilingDetectionDistance))
  (bind _returnvalue_2 (Utilities|Operators|Add _returnvalue (Math|Vector|MakeVector 0.0 0.0 _ceilingdetectiondistance)))
  (if (Variables|Default|GetDo%20Sky%20Check)
    (if (Variables|Default|GetDo%20Sphere%20Trace)
      (bind (_outhit _returnvalue_3) (Collision|SphereTraceForObjects _returnvalue _returnvalue_2 (Variables|/Script/Engine.PointLightComponent|GetSourceRadius LightComponent) (Utilities|Array|MakeArray "ObjectTypeQuery1") false _array))
      (return _returnvalue_3)
      (else
        (bind (_outhit_2 _returnvalue_4) (Collision|LineTraceByChannel _returnvalue _returnvalue_2 "TraceTypeQuery1" false _array "ForOneFrame"))
        (return _returnvalue_4)))
    (else
      (return false))))

(fn RefreshLightActors
  (FarLights|GatherLightActorData)
  (Components|InstancedStaticMesh|ClearInstances (Variables|Default|GetInstancedStaticMesh))
  (FarLights|GenerateMeshIstances))

(fn SetISMComponentParameters
  (bind _instancedstaticmesh (Variables|Default|GetInstancedStaticMesh))
  (bind (_tint_r _tint_g _tint_b) (Math|Color|BreakColor (Variables|GetTint_R)))
  (bind (_flaretint_r _flaretint_g _flaretint_b) (Math|Color|BreakColor (Variables|GetFlareTint_R)))
  (Rendering|Material|SetMaterial _instancedstaticmesh 0 (Variables|Parameters|Technical|GetMaterial))
  (Rendering|Material|SetCustomPrimitiveDataVector4 _instancedstaticmesh 0 (Math|Vector4|MakeVector4 _tint_r _tint_g _tint_b (Variables|Parameters|GetFalloff)))
  (Rendering|Material|SetCustomPrimitiveDataVector4 _instancedstaticmesh 4 (Math|Vector4|MakeVector4 _flaretint_r _flaretint_g _flaretint_b (Variables|Parameters|Flare|GetFlareFalloff)))
  (Rendering|Material|SetCustomPrimitiveDataVector4 _instancedstaticmesh 8 (Math|Vector4|MakeVector4 (Variables|Parameters|Flare|GetFlareSize) (Variables|Parameters|GetSaturation))))

(fn SetISMCustomDataVector3
  (params (ISMComponent :type Instanced%20Static%20Mesh%20Component%20Object%20Reference) (InstanceIndex :type Integer) (CustomDataIndex :type Integer) (Vector :type Vector) (MarkRenderStateDirty :type Boolean))
  (Components|InstancedStaticMesh|SetCustomDataValue ISMComponent InstanceIndex CustomDataIndex (.x Vector) false)
  (Components|InstancedStaticMesh|SetCustomDataValue ISMComponent InstanceIndex (+ CustomDataIndex 1) (.y Vector) false)
  (Components|InstancedStaticMesh|SetCustomDataValue ISMComponent InstanceIndex (+ CustomDataIndex 2) (.z Vector) MarkRenderStateDirty))

(fn SetPerInstanceMaterialParameters
  (params (ISMComponent :type Instanced%20Static%20Mesh%20Component%20Object%20Reference) (InstanceIndex :type Integer) (LightActorCachedData :type S%20Far%20Light%20Actors%20Cached%20Data%20Structure))
  (bind _lightactorcacheddata (Variables|GetLightActorCachedData))
  (bind _lightactorclass_2_79fd611942b4aeaa5582ce8dbcdcf58d (Utilities|Struct|BreakSFarLightActorsCachedData _lightactorcacheddata))
  (bind _perclassoverrides (Variables|Parameters|GetPerClassOverrides))
  (bind (_value _returnvalue) (Utilities|Map|Find _perclassoverrides _lightactorclass_2_79fd611942b4aeaa5582ce8dbcdcf58d))
  (bind (_tint_16_4cb6c09d45621abe571cb4931c95ecea _saturation_6_aa0872894fc87289e6aa24803be210f1 _falloff_10_a3bb73b64f6fb8ca297b738cd9cf4b35 _zoffset_13_1def638945034bda187749970d12c90b) (Utilities|Struct|BreakSFarLightOverrideParams _value))
  (bind (_intensity_14_1ec15487456387b2c794babc19d612dd _lightcolor_13_66e69b414b788f396165e980214c3ece) (Utilities|Struct|BreakSFarLightActorsCachedData (Variables|GetLightActorCachedData)))
  (bind _lightactorcacheddata_2 (Variables|GetLightActorCachedData))
  (bind _attenuationradius_8_9dd871f04bd00c5003fc7e95840493c5 (Utilities|Struct|BreakSFarLightActorsCachedData _lightactorcacheddata_2))
  (bind _glowzoffset (Variables|Parameters|Technical|GetGlowZOffset))
  (bind _returnvalue_2 (* _attenuationradius_8_9dd871f04bd00c5003fc7e95840493c5 _glowzoffset))
  (bind _instanceindex (Variables|GetInstanceIndex))
  (bind _ismcomponent (Variables|GetISMComponent))
  (Utilities|FlowControl|Sequence
    (:then_0
      (Utility|SetISMCustomDataVector3 (Variables|GetISMComponent) (Variables|GetInstanceIndex) 0 (* (* _intensity_14_1ec15487456387b2c794babc19d612dd 0.02) (Math|Conversions|ToVector(LinearColor) _lightcolor_13_66e69b414b788f396165e980214c3ece))) ; CDI 0-2 stores LightColor * LightIntensity
    )
    (:then_1
      (if _returnvalue
        (Utility|SetISMCustomDataVector3 _ismcomponent _instanceindex 3 (Math|Conversions|ToVector(LinearColor) _tint_16_4cb6c09d45621abe571cb4931c95ecea)) ; CDI 3-5 stores Tint
        (Utility|SetISMCustomDataVector3 0 _ismcomponent _instanceindex 6 (Math|Vector|MakeVector _saturation_6_aa0872894fc87289e6aa24803be210f1 _falloff_10_a3bb73b64f6fb8ca297b738cd9cf4b35 (+ _zoffset_13_1def638945034bda187749970d12c90b _returnvalue_2)) true) ; CDI 6-8 stores Saturation/Falloff/ZOffset
        (else
          (Utility|SetISMCustomDataVector3 _ismcomponent _instanceindex 3 "1.000000,1.000000,1.000000")
          (Utility|SetISMCustomDataVector3 0 _ismcomponent _instanceindex 6 (Math|Vector|MakeVector 1.0 1.0 _returnvalue_2) true) ; CDI 6-8 stores Saturation/Falloff/ZOffset
        )))))

(fn ConstructionScript
  (Components|Box|SetBoxExtent (Variables|Default|GetBox) (Variables|Parameters|GetVolumeSize) true)
  (FarLights|SetISMComponentParameters)
  (FarLights|GenerateMeshIstances))
